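# ButterCake music SoundCue that layers a docile loop and an aggressive loop for adaptive NPC music.
# Both loops feed one mixer through continuous modulators, with volume driven by the Unaware and Aggressive parameters.
# Each parameter defaults to zero, so a layer stays silent until its parameter is set.
# The cue uses the NPC music sound class and has a send to the music audio-analysis submix.
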SoundCue /ButterCake/Audio/Music/ButterCake_Music_Cue
root=Mixer_1
nodes [
 {id=Mixer_1, children [ModulatorContinuous_0, ModulatorContinuous_3], properties {inputVolume [1,1]}}
 {id=ModulatorContinuous_0, children [WavePlayer_0], properties {volumeModulationParams {default=0, parameterName=Unaware}}}
 {id=ModulatorContinuous_3, children [WavePlayer_2], properties {volumeModulationParams {default=0, parameterName=Aggressive}}}
 {id=WavePlayer_0, properties {bLooping=true, soundWaveAssetPtr=&/Buttercake_Music_Loop_Docile}}
 {id=WavePlayer_2, properties {bLooping=true, soundWaveAssetPtr=&/Buttercake_Music_Loop_Aggro}}
]
properties {
 maxDistance=2097152
 soundClassObject=&Fort_Music_NPC
 volumeMultiplier=0.531
 soundSubmixSends [{minSendLevel=0, soundSubmix=&Fortnite_AudioAnalysis_Music}]
}
