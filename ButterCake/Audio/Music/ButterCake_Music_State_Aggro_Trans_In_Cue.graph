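# ButterCake music cue for the transition into the Aggro state.
# A Random node picks one of three Aggro intro waves with equal weights and is excluded from branch culling.
# The cue plays in the Fort_Music_NPC sound class at reduced volume with a very large max distance and has a send to the music audio-analysis submix.
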SoundCue /ButterCake/Audio/Music/ButterCake_Music_State_Aggro_Trans_In_Cue
root=Random_0
nodes [
 {id=Random_0, children [WavePlayer_3,WavePlayer_4,WavePlayer_5], properties {bShouldExcludeFromBranchCulling=true, weights [1,1,1]}}
 {id=WavePlayer_3, properties {soundWaveAssetPtr=&/Buttercake_Music_Loop_Aggro_Intro_01}}
 {id=WavePlayer_4, properties {soundWaveAssetPtr=&/Buttercake_Music_Loop_Aggro_Intro_02}}
 {id=WavePlayer_5, properties {soundWaveAssetPtr=&/Buttercake_Music_Loop_Aggro_Intro_03}}
]
properties {
 maxDistance=2097152
 soundClassObject=&Fort_Music_NPC
 volumeMultiplier=0.531
 soundSubmixSends [{minSendLevel=0, soundSubmix=&Fortnite_AudioAnalysis_Music}]
}
