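# Sound cue for ButterCake's music as it transitions out of the aggro state.
# It plays a single Aggro outro wave directly from a wave player, with no randomization or looping nodes.
# It uses the NPC music sound class at reduced volume, has a very large max distance, and sends to the music audio-analysis submix.
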SoundCue /ButterCake/Audio/Music/ButterCake_Music_State_Aggro_Trans_Out_Cue
root=WavePlayer_3
nodes [{id=WavePlayer_3, properties {soundWaveAssetPtr=&/Buttercake_Music_Loop_Aggro_Outro_04}}]
properties {
 maxDistance=2097152
 soundClassObject=&Fort_Music_NPC
 volumeMultiplier=0.531
 soundSubmixSends [{minSendLevel=0, soundSubmix=&Fortnite_AudioAnalysis_Music}]
}
