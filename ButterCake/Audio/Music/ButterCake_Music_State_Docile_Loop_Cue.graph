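# Sound cue that plays the docile-state music loop for ButterCake.
# A single looping wave player feeds the output directly, with no mixing, modulation or randomization.
# The cue routes to the NPC music sound class at reduced volume and carries a very large max audible distance.
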SoundCue /ButterCake/Audio/Music/ButterCake_Music_State_Docile_Loop_Cue
root=WavePlayer_1
nodes [{id=WavePlayer_1, properties {bLooping=true, soundWaveAssetPtr=&/Buttercake_Music_Loop_Docile}}]
properties {maxDistance=2097152, soundClassObject=&Fort_Music_NPC, volumeMultiplier=0.531}
