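; Quartz music track component for ButterCake that plays aggro start and end stingers from gameplay data tracker events.
; A change in whether event values exist can trigger the start stinger, while tracker updates trigger the end stinger once aggro reaches zero, with flags limiting repeats.
; Both play quantized to the beat.
; Start volume follows the Aggressive value; end volume follows the greater of Aggressive and Unaware, clamped to 0–1.
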(event Override|EventOnDataTrackerUpdate
  (|CallParentFunction)
  (Variables|Default|SetAggroValue (Class|FortGameplayDataTrackerComponent|GetEventValue (CallFunction|GetDataTrackerComponent) (Variables|Default|GetAggressiveEvent%20Tag)))
  (Variables|Default|SetUnwareValue (Class|FortGameplayDataTrackerComponent|GetEventValue (CallFunction|GetDataTrackerComponent) (Variables|Default|GetUnawareEvent%20Tag)))
  (Utilities|FlowControl|Sequence
    (:then_0
      (if (<= (Variables|Default|GetAggroValue) 0)
        (if (not (Variables|Default|GetbAggroEndPlayed))
          (CallFunction|OnAggroEnd))))
    (:then_1
      (CallFunction|UpdateStingerVolumes))))

(event Internal|EventOnChangedHasAnyEventValue
  (params (DataTrackerComponent :type Fort%20Gameplay%20Data%20Tracker%20Component%20Object%20Reference) (bHasEventValues :type Boolean))
  (|CallParentFunction DataTrackerComponent bHasEventValues)
  (bind _output_get (Variables|Default|SetAggroValue (Class|FortGameplayDataTrackerComponent|GetEventValue (CallFunction|GetDataTrackerComponent) (Variables|Default|GetAggressiveEvent%20Tag))))
  (if (> _output_get 0)
    (CallFunction|OnAggroStart)
    (else
      (Variables|Default|SetbAggroStartPlayed false)
      (Variables|Default|SetbAggroEndPlayed true))))

(event Custom|OnAggroStart
  (if (> (Variables|Default|GetAggroValue) 0)
    (if (not (Variables|Default|GetbAggroStartPlayed))
      (bind _returnvalue (Audio|CreateSound2D (Variables|Default|GetStartAggroStinger) 0.0))
      (bind _output_get (Variables|Default|SetStartAggroComponent _returnvalue))
      (Audio|Components|Audio|PlayQuantized _output_get (CallFunction|GetClockHandle) (Utilities|Struct|MakeQuartzQuantizationBoundary "Beat"))
      (Variables|Default|SetbAggroStartPlayed true)
      (Variables|Default|SetbAggroEndPlayed false)
      (CallFunction|UpdateStingerVolumes))))

(event Custom|OnAggroEnd
  (bind _returnvalue (Audio|CreateSound2D (Variables|Default|GetEndAggroStinger)))
  (bind _output_get (Variables|Default|SetEndAggroComponent _returnvalue))
  (Audio|Components|Audio|PlayQuantized _output_get (CallFunction|GetClockHandle) (Utilities|Struct|MakeQuartzQuantizationBoundary "Beat"))
  (Variables|Default|SetbAggroEndPlayed true)
  (CallFunction|UpdateStingerVolumes))

(fn OnRep_AggroValue)

(fn UpdateStingerVolumes
  (bind _startaggrocomponent (Variables|Default|GetStartAggroComponent)
    (:then
      (Audio|Components|Audio|SetVolumeMultiplier _startaggrocomponent (Math|Float|MapRangeClamped (Variables|Default|GetAggroValue) 0.0 1.0 0.0 1.0) :!class /Script/Engine.AudioComponent))
    (:else
      _))
  (bind _endaggrocomponent (Variables|Default|GetEndAggroComponent)
    (:then
      (Audio|Components|Audio|SetVolumeMultiplier _endaggrocomponent (Math|Float|MapRangeClamped (Math|Float|SelectFloat (Variables|Default|GetAggroValue) (Variables|Default|GetUnwareValue) (>= (Variables|Default|GetAggroValue) (Variables|Default|GetUnwareValue))) 0.0 1.0 0.0 1.0) :!class /Script/Engine.AudioComponent))
    (:else)))
